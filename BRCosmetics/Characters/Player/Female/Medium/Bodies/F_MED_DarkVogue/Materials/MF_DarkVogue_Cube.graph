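# Material function that overlays a glowing "Cube" glyph look on the DarkVogue female outfit's material attributes, blended by an Alpha input.
# Glyphs sample UVs or pre-skinned triplanar coordinates, modulated by panning masks, pulses and Fresnel, then combine with optional core glow and noise-based crystal edges.
# Static switches select emissive-plus-base-color or full surface override, a mobile fallback, and CubeStageINT-driven "Kevin C" intensity stages.
# Hue shift and desaturation tint the glow.
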MaterialFunction /BRCosmetics/Characters/Player/Female/Medium/Bodies/F_MED_DarkVogue/Materials/MF_DarkVogue_Cube
graph {nodes [
  Constant() {desc=Glyph Panner Independent Scale, r=0.125}
  n1 = FunctionInput() {inputType=FunctionInput_MaterialAttributes, sortPriority=1}
  n2 = LinearInterpolate(
   A=TextureCoordinate()
   B=TextureCoordinate() {coordinateIndex=1}
   Alpha=Round(None=ScalarParameter() {group=T H E C U B E, parameterName=Use 2nd UV Channel, sortPriority=250})
  )
  n3 = ScalarParameter() {defaultValue=5, desc=5, group=T H E C U B E, parameterName=Overall Texture Tiling}
  n4 = Multiply(A=@n3, B=Multiply(A=Constant() {r=2}, B=@n2))
  n5 = ComponentMask(None=VectorParameter() {group=T H E C U B E, parameterName=GlyphsOffset, defaultValue=<rgba 0 0 0 1>}.RGB) {g=true, r=true}
  n6 = StaticBoolParameter() {group=T H E C U B E, parameterName=Use Triplanar Glyph Tiling, sortPriority=5000}
  n7 = VertexInterpolator(VS=PreSkinnedPosition())
  n8 = Multiply(A=@n3, B=Divide(A=@n7.PS, B=Constant() {desc=11.5, r=55}))
  n9 = TextureObject() {samplerType=SAMPLERTYPE_LinearColor, texture=/Game/Athena/Prototype/Blueprints/Cube/Materials/DFHeightMask}
  n10 = VertexInterpolator(VS=PreSkinnedNormal())
  n11 = ComponentMask(None=@n10.PS) {r=true}
  n12 = Constant() {r=1}
  n13 = ComponentMask(None=MaterialFunctionCall(In=Abs(None=@n11), Contrast=@n12) {materialFunction=&CheapContrast}.Result) {r=true}
  n14 = LinearInterpolate(
   A=TextureSample(UVs=ComponentMask(None=@n8) {b=true, r=true}, Tex=@n9) {samplerSource=SSM_Wrap_WorldGroupSettings}.RGB
   B=TextureSample(UVs=ComponentMask(None=@n8) {b=true, g=true}, Tex=@n9) {samplerSource=SSM_Wrap_WorldGroupSettings}.RGB
   Alpha=@n13
  ) {desc=Float3}
  n15 = TextureSample(UVs=ComponentMask(None=@n8) {g=true, r=true}, Tex=@n9) {samplerSource=SSM_Wrap_WorldGroupSettings}
  n16 = ComponentMask(None=@n10.PS) {b=true}
  n17 = ComponentMask(None=MaterialFunctionCall(In=Abs(None=@n16), Contrast=@n12) {materialFunction=&CheapContrast}.Result) {r=true}
  n18 = StaticSwitch(
   True=MaterialFunctionCall(Float3=LinearInterpolate(A=@n14, B=@n15.RGB, Alpha=@n17)) {materialFunction=&BreakOutFloat3Components}.G
   False=TextureSample(UVs=Add(A=@n4, B=@n5)) {samplerType=SAMPLERTYPE_LinearColor, texture=&DFHeightMask}.R
   Value=@n6
  )
  n19 = Power(Base=@n18, Exp=ScalarParameter() {defaultValue=6, group=T H E C U B E, parameterName=Glyph Thickness})
  n20 = ScalarParameter() {defaultValue=0.125, group=T H E C U B E, parameterName=PannerMask_TilingMultiplier, sortPriority=50}
  n21 = Divide(A=@n5, B=@n20)
  n22 = ScalarParameter() {group=T H E C U B E, parameterName=IsPickaxe, sortPriority=50}
  n23 = Panner(
   Coordinate=Multiply(A=@n3, B=Multiply(A=@n20, B=@n2))
   Time=Multiply(A=Constant() {desc="Glyph Panner Speed ", r=0.1}, B=Time())
   Speed=AppendVector(
    A=ScalarParameter() {group=T H E C U B E, parameterName=Glyph Panner Speed X, sortPriority=50}
    B=ScalarParameter() {defaultValue=1, group=T H E C U B E, parameterName=Glyph Panner Speed Y, sortPriority=50}
   )
  ) {speedY=1}
  n24 = TextureSampleParameter2D(UVs=Add(A=LinearInterpolate(A=@n5, B=@n21, Alpha=@n22), B=@n23)) {
   group=T H E C U B E
   parameterName=PannerMask
   samplerSource=SSM_Wrap_WorldGroupSettings
   sortPriority=50
   texture=/BRCosmetics/Characters/Player/Female/Medium/Bodies/F_MED_DarkVogue/Textures/T_DarkVogue_CubeMask
  }
  n25 = Constant() {desc=-180, r=-835}
  n26 = Time()
  n27 = VectorParameter() {group=T H E C U B E, parameterName=Triplanar Glyph Panner Speed, sortPriority=5001, defaultValue=<rgba 0 0 0.075 0>}
  n28 = Add(A=Multiply(A=@n3, B=Divide(A=@n7.PS, B=@n25)), B=Frac(None=Multiply(A=@n26, B=@n27.RGB)))
  n29 = Constant() {r=2}
  n30 = TextureObject() {samplerType=SAMPLERTYPE_LinearColor, texture=&DFHeightMask}
  n31 = TextureSample!(UVs=ComponentMask(None=@n28) {b=true, r=true}, Tex=@n30, Bias=@n29, Apply View MipBias) {mipValueMode=TMVM_MipBias, samplerSource=SSM_Wrap_WorldGroupSettings}
  n32 = TextureSample(UVs=ComponentMask(None=@n28) {b=true, g=true}, Tex=@n30, Bias=@n29) {mipValueMode=TMVM_MipBias, samplerSource=SSM_Wrap_WorldGroupSettings}
  n33 = ComponentMask(None=@n28) {g=true, r=true}
  n34 = MaterialFunctionCall(Float3=LinearInterpolate(
    A=LinearInterpolate(A=@n31.RGB, B=@n32.RGB, Alpha=@n13) {desc=Float3}
    B=TextureSample(UVs=@n33, Tex=@n30, Bias=@n29) {mipValueMode=TMVM_MipBias, samplerSource=SSM_Wrap_WorldGroupSettings}.RGB
    Alpha=@n17
   )) {materialFunction=&BreakOutFloat3Components}
  n35 = Saturate(None=LinearInterpolate(
    A=Constant() {desc=Glyph Panner Opacity Max, r=1}
    B=ScalarParameter() {group=T H E C U B E, parameterName=Glyph Panner Opacity Min}
    Alpha=StaticSwitch(True=@n34.B, False=@n24.B, Value=@n6)
   ))
  n36 = Saturate(None=Multiply(A=Multiply(A=@n18, B=@n35), B=Constant() {r=1.25}))
  n37 = VectorParameter() {group=T H E C U B E, parameterName=Glyph Glow Color, defaultValue=<rgba 0.4689 0.042 0.8958 1>}
  n38 = VectorParameter() {group=T H E C U B E, parameterName=Glyph Glow Color - KEvin C Max, defaultValue=<rgba 0.4689 0.042 0.8958 1>}
  n39 = Divide(A=ScalarParameter() {parameterName=CubeStageINT}) {constB=4}
  n40 = Saturate(None=Subtract(A=@n39) {constB=2})
  n41 = StaticBoolParameter() {group=T H E C U B E, parameterName=Use Kevin C Incements}
  n42 = StaticSwitch(True=LinearInterpolate(A=@n37.RGB, B=@n38.RGB, Alpha=@n40), False=@n37.RGB, Value=@n41)
  n43 = Multiply(A=LinearInterpolate(B=@n19, Alpha=@n36), B=@n42)
  n44 = OneMinus(None=Fresnel(ExponentIn=ScalarParameter() {defaultValue=0.5, group=T H E C U B E, parameterName=Core Glow Simple Fresnel Ex}))
  n45 = ScalarParameter() {defaultValue=2, group=T H E C U B E, parameterName=Core Glow Simple Radius}
  n46 = StaticSwitch(
   True=LinearInterpolate(
    A=ScalarParameter() {defaultValue=2.5, group=T H E C U B E, parameterName=Emissive Fresnel Power}
    B=ScalarParameter() {defaultValue=0.5, group=T H E C U B E, parameterName=Emissive Fresnel Power - KEVINCMAX}
    Alpha=@n40
   )
   False=ScalarParameter() {defaultValue=2.5, group=T H E C U B E, parameterName=Emissive Fresnel Power}
   Value=@n41
  )
  n47 = FunctionInput() {bUsePreviewValueAsDefault=true, inputName=AO, inputType=FunctionInput_Scalar, sortPriority=3, previewValue=<1 1 1 1>}
  n48 = OneMinus(None=@n47)
  n49 = Constant()
  n50 = GetMaterialAttributes(None=@n1) {attributeGetTypes [#1, #2, #3]}
  n51 = OneMinus(None=Fresnel(
    ExponentIn=ScalarParameter() {defaultValue=0.5, group=T H E C U B E, parameterName=Emissive Fresnel Exponent}
    BaseReflectFractionIn=StaticSwitchParameter(True=@n48, False=@n49) {group=T H E C U B E, parameterName=Use AO Mask in Fresnel}
    Normal=Transform(None=@n50.Normal)
   ))
  n52 = MaterialFunctionCall(Value=Multiply(A=Time(), B=Constant() {desc=Glyph Pulse Speed, r=0.3})) {materialFunction=&LinearSine}
  n53 = LinearInterpolate(
   A=ScalarParameter() {defaultValue=6, group=T H E C U B E, parameterName=Glpyh Pulse Flare Width Max}
   B=ScalarParameter() {defaultValue=25, group=T H E C U B E, parameterName=Glpyh Pulse Flare Width Min}
   Alpha=@n52."Rounded Linear Sine "
  )
  n54 = StaticBoolParameter() {defaultValue=true, group=T H E C U B E, parameterName=Use Pulsing Glyphs}
  n55 = StaticSwitch(True=Multiply(A=Power(Base=@n18, Exp=@n53), B=Constant() {r=5}), False=@n19, Value=@n54)
  n56 = Multiply(A=@n42, B=Saturate(None=LinearInterpolate(B=@n55, Alpha=@n36) {constB=0}))
  n57 = ScalarParameter() {defaultValue=0.8, group=T H E C U B E, parameterName=Overall Emissive Min}
  n58 = Saturate(None=@n39)
  n59 = StaticSwitch(
   True=LinearInterpolate(B=ScalarParameter() {defaultValue=5, group=T H E C U B E, parameterName=Overall Emissive Max - KevinCMAX}, Alpha=@n58) {constB=5}
   False=ScalarParameter() {defaultValue=1, group=T H E C U B E, parameterName=Overall Emissive Max}
   Value=@n41
  )
  n60 = LinearInterpolate(B=ScalarParameter() {defaultValue=2, group=T H E C U B E, parameterName=Overall Emissive Pulse Speed - KEVIN C}, Alpha=@n40)
  n61 = ScalarParameter() {defaultValue=2.5, group=T H E C U B E, parameterName=Overall Emissive Pulse Speed}
  n62 = Sine(None=Multiply(A=Time(), B=StaticSwitch(True=@n60, False=@n61, Value=@n41)))
  n63 = StaticSwitchParameter(True=Multiply(A=LinearInterpolate(A=@n57, B=@n59, Alpha=@n62), B=@n56), False=@n56) {defaultValue=true, group=T H E C U B E, parameterName=Use Overall Pulsing Effect}
  n64 = StaticSwitchParameter(True=Multiply(A=Power(Base=@n51, Exp=@n46), B=@n63), False=@n63) {defaultValue=true, group=T H E C U B E, parameterName=Use Fresnel as part of Emissive Mask}
  n65 = Add(
   A=Multiply(
    A=Power(Base=@n44, Exp=@n45)
    B=VectorParameter() {group=T H E C U B E, parameterName=Core Glow Color, defaultValue=<rgba 0.1478 0 0.2917 1>}.RGB
   )
   B=@n64
  )
  n66 = VectorParameter() {group=T H E C U B E, parameterName=Core Glow Color 1, defaultValue=<rgba 0.008 0.0012 0.0307 1>}
  n67 = VectorParameter() {group=T H E C U B E, parameterName=Core Glow Color 2, defaultValue=<rgba 0.0046 0.0001 0.026 1>}
  n68 = VectorParameter() {group=T H E C U B E, parameterName=Core Glow Color 3, defaultValue=<rgba 0.1478 0 0.2917 1>}
  n69 = Constant() {desc=Core Glow Mote Texture Independent Scale, r=0.1}
  n70 = TextureCoordinate()
  n71 = TextureSample!(UVs=Multiply(A=Multiply(A=@n69, B=@n70), B=@n3), Tex, Apply View MipBias) {samplerSource=SSM_Wrap_WorldGroupSettings, samplerType=SAMPLERTYPE_Masks, texture=/Game/Global/Textures/Default/Noise/T_CubeNoise}
  n72 = ScalarParameter() {group=T H E C U B E, parameterName=Core Glow Blend - Kevin C Min}
  n73 = ScalarParameter() {defaultValue=1, group=T H E C U B E, parameterName=Core Glow Blend - Kevin C Max}
  n74 = Power(
   Base=Multiply(A=@n71.B, B=@n51)
   Exp=StaticSwitch(
    True=LinearInterpolate(A=@n72, B=@n73, Alpha=@n58)
    False=ScalarParameter() {defaultValue=1, group=T H E C U B E, parameterName=Core Glow Blend}
    Value=@n41
   )
  )
  n75 = Add(A=@n52."Rounded Linear Sine ", B=Constant() {r=1})
  n76 = Saturate(None=StaticSwitch(True=Multiply(A=@n74, B=@n75), False=@n74, Value=@n54))
  n77 = Add(
   A=Multiply(
    A=ScalarParameter() {defaultValue=0.5, group=T H E C U B E, parameterName=Core Glow Power}
    B=MaterialFunctionCall(1=@n66.RGB, 2=@n67.RGB, 3=@n68.RGB, A=@n76) {materialFunction=&Lerp_Multiple_Float3}.Lerp 3 Inputs
   )
   B=@n64
  )
  n78 = Saturate(None=Multiply(A=@n47, B=@n47))
  n79 = Constant() {r=2.5}
  n80 = Constant() {desc=Edge Texture Independent Scale, r=0.8}
  n81 = TextureCoordinate()
  n82 = TextureSample(UVs=Multiply(A=Multiply(A=@n80, B=@n81), B=@n3)) {samplerSource=SSM_Wrap_WorldGroupSettings, samplerType=SAMPLERTYPE_Masks, texture=&T_CubeNoise}
  n83 = Fresnel(ExponentIn=Constant() {desc=Edge Texture Fresnel, r=1.2})
  n84 = Saturate(None=LinearInterpolate(A=Constant() {r=-0.35}, B=Constant() {r=1.12}, Alpha=OneMinus(None=@n83)))
  n85 = LinearInterpolate(A=Power(Base=Multiply(A=@n79, B=@n82.R), Exp=Constant() {r=2}), Alpha=@n84)
  n86 = Multiply(
   A=QualitySwitch(Default=@n78, Low=Constant() {r=1}, High=@n78, Medium=@n78)
   B=StaticSwitchParameter(True=Multiply(A=@n84, B=@n85), False=@n84) {defaultValue=true, group=T H E C U B E, parameterName=Use Cube Crystal in Material}
  )
  n87 = LinearInterpolate(
   B=StaticSwitchParameter(True=StaticSwitchParameter(True=@n65, False=@n77) {group=T H E C U B E, parameterName=Use Simple Core Glow}, False=@n64) {group=T H E C U B E, parameterName=Core Glow}
   Alpha=Add(A=@n86, B=Constant() {desc=Edge Texture Emissive Influence, r=0.5})
  )
  n88 = StaticSwitchParameter(True=QualitySwitch(Default=@n43, Low=@n43, High=@n43, Medium=@n43), False=QualitySwitch(Default=@n87, Low=@n43, High=@n87, Medium=@n87)) {group=T H E C U B E, parameterName=TEST Mobile Override}
  n89 = StaticSwitchParameter(True=Add(A=@n50.EmissiveColor, B=@n88), False=@n88) {
   desc=Adds the settings in the 'Emissive' section above to the emissive output by this effect. Does not clamp.
   group=T H E C U B E
   parameterName=Add 'Emissive' Texture
  }
  n90 = MaterialFunctionCall(Hue Shift Percentage=ScalarParameter() {defaultValue=1, group=T H E C U B E, parameterName=Glyphs_HueShift}, Texture=@n89) {materialFunction=&HueShift}
  n91 = Multiply(
   A=ScalarParameter() {defaultValue=1, group=T H E C U B E, parameterName=Glyphs_EmissiveMultipler}
   B=Desaturation(None=@n90.Result, Fraction=ScalarParameter() {group=T H E C U B E, parameterName=GlyphsDesaturation})
  )
  n92 = Desaturation(None=@n90.Result, Fraction=ScalarParameter() {group=T H E C U B E, parameterName=GlyphsDesaturation_BaseColor})
  n93 = ScalarParameter() {defaultValue=1, group=T H E C U B E, parameterName=Glyphs_BaseColorMultipler}
  n94 = SetMaterialAttributes!(MaterialAttributes=@n50.MaterialAttributes, EmissiveColor=@n91, BaseColor=Add(A=@n50.BaseColor, B=Multiply(A=@n92, B=@n93))) {attributeSetTypes [#2, #3]}
  n95 = SetMaterialAttributes!(
   MaterialAttributes=@n50.MaterialAttributes
   BaseColor=LinearInterpolate(
    A=VectorParameter() {group=T H E C U B E, parameterName=Edge Diffuse, defaultValue=<rgba 0.0053 0.0017 0.0156 1>}.RGB
    B=VectorParameter() {group=T H E C U B E, parameterName=Base Diffuse, defaultValue=<rgba 0.0061 0.0029 0.0469 1>}.RGB
    Alpha=@n86
   )
   Specular=LinearInterpolate(A=Constant() {desc=Edge Specular, r=0.1}, B=Constant() {desc=Body Specular, r=0.7}, Alpha=@n86)
   Metallic=Constant() {desc=Edge Metallic}
   Roughness=LinearInterpolate(A=Constant() {desc=Edge Roughness, r=0.8}, B=Constant() {desc=Body Roughness, r=0.6}, Alpha=@n86)
   EmissiveColor=@n91
  ) {attributeSetTypes [
    #3
    #4
    #5
    #6
    #2
   ]}
  FunctionOutput(None=BlendMaterialAttributes(
    A=@n1
    B=StaticSwitchParameter(True=@n94, False=@n95) {group=T H E C U B E, parameterName=Affects Emissive Only}
    Alpha=FunctionInput() {bUsePreviewValueAsDefault=true, inputName=Alpha, inputType=FunctionInput_Scalar, sortPriority=2}
   ))
  StaticSwitch()
  TextureSample!(UVs=@n23, Tex, Bias=Constant() {r=2}, Apply View MipBias) {
   mipValueMode=TMVM_MipBias
   samplerSource=SSM_Wrap_WorldGroupSettings
   samplerType=SAMPLERTYPE_LinearColor
   texture=&DFHeightMask
  }
  VectorParameter() {group=T H E C U B E, parameterName=Base Diffuse, defaultValue=<rgba 0.0061 0.0029 0.0469 1>}
 ]}
