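# Parent material for the DarkVogue skin of the StreetFashionRed pickaxe, built on material attributes for skeletal meshes with the shading model set by expression.
# It wraps a skin-specific effect between the shared 2023 character parent PreFX and PostFX functions.
# The PreFX result feeds MF_DarkVogue_Cube, borrowed from the DarkVogue character body, before reaching PostFX.
# A static component mask parameter picks which Tech Art Mask channel drives the Cube effect's alpha, defaulting to red.
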Material /BRCosmetics/Weapons/FORT_Melee/Pickaxe_StreetFashionRed/Skins/DarkVogue/Materials/M_Pickaxe_DarkVogue_PARENT
graph {
 properties {
  bUseMaterialAttributes=true
  bUsedWithMeshDeformer=true
  bUsedWithSkeletalMesh=true
  pixelDepthOffsetMode=PDOM_Legacy
  shadingModel=MSM_FromMaterialExpression
  usedShadingModels=Default Lit | Subsurface | Clear Coat | Subsurface Profile | Cloth
 }
 nodes [
  n1 = MaterialFunctionCall() {materialFunction=&MF_CharacterParent_2023_PreFX}
  MaterialFunctionCall(
   Material=MaterialFunctionCall!(
    In=@n1.Result
    Alpha=StaticComponentMaskParameter(None=@n1.Tech Art Mask) {defaultR=true, group=Skin FX, parameterName="Channel Embracing T H E C U B E "}
    AO
   ) {materialFunction=/BRCosmetics/Characters/Player/Female/Medium/Bodies/F_MED_DarkVogue/Materials/MF_DarkVogue_Cube}.Result
   M=@n1.M
   Pre Skinned Local Position=@n1.M
   Distance Blend=@n1.M
   Fresnel=@n1.M
  ) {materialFunction=&MF_CharacterParent_2023_PostFX}
 ]
}
